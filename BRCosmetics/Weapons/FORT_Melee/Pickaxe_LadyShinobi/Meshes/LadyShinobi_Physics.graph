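# Physics asset for the LadyShinobi pickaxe, a two-handed melee weapon cosmetic.
# A single box body on the Melee_TwoHanded_Handle bone approximates the whole weapon as one long, tilted shape.
# It has no constraints, so the weapon simulates and collides as one rigid piece.
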PhysicsAsset /BRCosmetics/Weapons/FORT_Melee/Pickaxe_LadyShinobi/Meshes/LadyShinobi_Physics
bodies {Melee_TwoHanded_Handle [{shapeType=Box, extentX=13.4152, extentY=11.5004, extentZ=117.3749, center=<0.3502 1.6397 40.8548>, rotation=<rot 3.3818 -90 -0.1013>}]}
constraints []
